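; Base Blueprint for a CherrySmoke hero stand, a searchable loot container that displays a hero item mesh.
; On BeginPlay it binds an empty interact handler and, as a marked band-aid, re-applies its loot tier group and clears the searched flag once playlist data is ready.
; Hiding openable cosmetics hides and destroys the item mesh.
; On loot, the static mesh ignores GameTraceChannel1 so loot bouncing back inside stays lootable.
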(event EventBeginPlay
  (|CallParentFunction)
  (Interaction|BindEventtoOnInteract (EventDispatchers|CreateEvent :!event Custom|OnInteract_Event))
  ; Band-aid solution @Kyle.Cavasos
  (Class|FortAsyncActionCurrentPlaylistReady|PlaylistDataReadyAsync
    (:OnReady
      (Container|SetSearchLootTierGroup (Variables|Container|GetSearchLootTierGroup))
      (Variables|Container|SetbAlreadySearched false))))

(event Custom|OnInteract_Event
  (params (SelfActor :type Building%20Actor%20Object%20Reference) (InteractingPawn :type Fort%20Pawn%20Object%20Reference)))

(event EventHideChestOpenableCosmetics
  (bind _heroitemmesh (Variables|Default|GetHeroItemMesh))
  (|CallParentFunction)
  (Development|SetHiddeninGame _heroitemmesh true true)
  (Components|DestroyComponent _heroitemmesh))

(event EventOnLoot
  (|CallParentFunction)
  (Collision|SetCollisionResponsetoChannel (Variables|Debug|GetStaticMeshComponent) "ECC_GameTraceChannel1" "ECR_Ignore") ; Fix issue where loot that bounces back into the container cannot be looted
)

(fn ConstructionScript
  (|CallParentFunction))
